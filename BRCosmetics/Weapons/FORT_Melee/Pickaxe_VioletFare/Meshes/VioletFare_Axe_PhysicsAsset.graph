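# Physics asset for the VioletFare pickaxe mesh, a two-handed melee weapon cosmetic.
# It holds a single capsule body on the Melee_TwoHanded_Handle bone, rotated to lie along the weapon.
# It has no constraints, so the axe is treated as one rigid collision shape.
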PhysicsAsset /BRCosmetics/Weapons/FORT_Melee/Pickaxe_VioletFare/Meshes/VioletFare_Axe_PhysicsAsset
bodies {Melee_TwoHanded_Handle [{shapeType=Capsule, length=74.6904, radius=15.0039, center=<-0.278 -0.0472 -0.7388>, rotation=<rot 1.6555 -90 -0.0003>}]}
constraints []
